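# Niagara system for the MuffinLadle guitar cosmetic that adds a persistent red glow and light for its strings.
# Glow bursts six sprites, using partition index to pick each one's offset from an array, forming a row of vertically aligned streaks.
# Light bursts an offset red light-renderer particle; both emitters keep particles alive until deactivation.
# LWE_Strings shows no modules or renderers, and the system exposes no user parameters.
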NiagaraSystem /SparksCosmetics/Instrument/Guitar/MuffinLadle/Effects/NS_Guitar_MuffinLadle_Strings
dependencies {
 usedDataInterfaces [&NiagaraDataInterfaceArrayFloat3]
 usedDynamicInputs []
 usedModules [&SystemState]
}
emitters {
 Glow {
  emitterUpdateScript [
   &EmitterState()
   &SpawnBurst_Instantaneous(Spawn Count=6, Use Spawn Probability=false, UseLoopCountLimit=true)
  ]
  particleSpawnScript [
   &PartitionParticles(Number Of Partitions=6)
   &InitializeParticle(
    Color=<rgba 1 0.0365 0.0365 1>
    Position Offset=&SelectVectorFromArray(
     Direct Array Index=$Particles.PartitionParticles.Partition
     Vector Selection Array=DataInterface() {propertyValues {FloatData [<-2.25 1.75 -19>, <-1.35 1.75 -19>, <-0.45 1.75 -19>, <0.45 1.75 -19>, <1.35 1.75 -19.5>, <2.25 1.75 -19.5>]}}
    )
    Sprite Size=&Multiply_Vector2D(A=<0.5 4>, B=&Vector2DFromFloat(Value=2))
    Sprite Size Mode=::NewEnumerator4
    UsePositionOffset=true
    Write Material Random=true
   )
  ]
  particleUpdateScript [
   &ParticleState(
    Kill Particles When Lifetime Has Elapsed=false
    Let Infinitely Lived Particles Die When Emitter Deactivates=true
    Loop Particles Lifetime=true
   )
   &SpriteFacingAndAlignment(Sprite Alignment=<0 0 1>, WriteAlignment=true, WriteFacing=false)
  ]
  rendererClasses [&NiagaraSpriteRendererProperties]
  renderers [{rendererClass=&NiagaraSpriteRendererProperties}]
 }
 LWE_Strings {rendererClasses [], renderers []}
 Light {
  emitterUpdateScript [
   &EmitterState()
   &SpawnBurst_Instantaneous(Use Spawn Probability=false, UseLoopCountLimit=true)
  ]
  particleSpawnScript [
   /SetVariables(Particles.LightRadius=30) {isSetParameters=true}
   &InitializeParticle(
    Color=<rgba 1 0.0417 0.0417 1>
    Position Offset=<0 15 5>
    Sprite Size Mode=::NewEnumerator3
    Uniform Sprite Size=$Particles.LightRadius
    UsePositionOffset=true
    Write Material Random=true
   )
  ]
  particleUpdateScript [&ParticleState(
    Kill Particles When Lifetime Has Elapsed=false
    Let Infinitely Lived Particles Die When Emitter Deactivates=true
    Loop Particles Lifetime=true
   )]
  rendererClasses [&NiagaraLightRendererProperties]
  renderers [{rendererClass=&NiagaraLightRendererProperties, rendererIndex=0}]
 }
}
summary {
 systemName=NS_Guitar_MuffinLadle_Strings
 emitters [
  {emitterName=LWE_Strings, rendererClasses []}
  {emitterName=Light, rendererClasses [&NiagaraLightRendererProperties]}
  {emitterName=Glow, rendererClasses [&NiagaraSpriteRendererProperties]}
 ]
 userVariables []
}
